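; Behavior tree injection evaluator that decides when an AI bot can trigger the Demon Mask Void alt use.
; It caches the bot's controller and Phoebe pawn on enter, and arms an alt-fire flag after a random scalable delay once the main-use behavior succeeds.
; Starting requires the lesser-demon toggle, a valid teleport BGA on the pawn's void component and that flag, which it consumes; ongoing conditions always pass.
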(event EventOnEnter
  (params (OwnerComp :type Behavior%20Tree%20Component%20Object%20Reference))
  (bind _ascontroller (Utilities|Casting|CastToController (Components|GetOwner OwnerComp)))
  (bind _asfort_athena_aibot_controller (Utilities|Casting|CastToFortAthenaAIBotController _ascontroller))
  (bind _output_get (Variables|Default|SetAIController _asfort_athena_aibot_controller))
  (bind _asbp_player_pawn_athena_phoebe (Utilities|Casting|CastToBP_PlayerPawn_Athena_Phoebe (Pawn|GetControlledPawn _output_get)))
  (Variables|Default|SetAIPawn _asbp_player_pawn_athena_phoebe))

(event EventOnBlackboardKeyValueChanged
  (params (OwnerComp :type Behavior%20Tree%20Component%20Object%20Reference) (ChangedKeyName :type Name :default "None")))

(event EventOnConcurentBehaviorExecutionStatusChanged
  (params (InjectedBehavior :type Fort%20Injected%20Behavior%20Object%20Reference) (PreviousExecutionStatus :type EExecutionStatus%20Enum :default "ExecutionError") (NewExecutionStatus :type EExecutionStatus%20Enum :default "ExecutionError"))
  (if (== (Utilities|GetClass InjectedBehavior) (Variables|Default|GetMainUseInjectedBehavior))
    (switch Utilities|FlowControl|Switch|SwitchonEExecutionStatus NewExecutionStatus
      (:ExecutionSuccess
        (Utilities|FlowControl|Delay (Math|Random|RandomFloatinRange (ScalableFloat|GetValueatLevel (Variables|Default|GetAltUseDelayMin_Hf)) (ScalableFloat|GetValueatLevel (Variables|Default|GetAltUseDelayMax_Hf))))
        (Variables|Default|SetCanUseAltFire true)))))

(fn EvaluateOngoingConditions
  (params (OwnerComp :type Behavior%20Tree%20Component%20Object%20Reference))
  (|CallParentFunction OwnerComp)
  (return true))

(fn EvaluateStartingConditions
  (params (OwnerComp :type Behavior%20Tree%20Component%20Object%20Reference))
  (bind _aipawn (Variables|Default|GetAIPawn))
  (bind _returnvalue (Actor|GetComponentbyClass _aipawn "/DemonMaskVoid/Gameplay/Components/B_DemonMask_Void_FortPawnComponent.B_DemonMask_Void_FortPawnComponent_C"))
  (|CallParentFunction OwnerComp)
  (if (ScalableFloat|GetValueasBool (Variables|Default|GetLesserDemonCanUseMask_Hf))
    (Utilities|IsValid (Variables|Default|GetAIController)
      (:Is%20Valid
        (Utilities|IsValid _returnvalue
          (:Is%20Valid
            (Utilities|IsValid (Variables|/Script/DemonMaskVoidRuntime.FortPawnComponent_VoidTeleport|GetBGAToTeleport _returnvalue)
              (:Is%20Valid
                (if (Variables|Default|GetCanUseAltFire)
                  (Variables|Default|SetCanUseAltFire false)
                  (return true)
                  (else
                    _)))
              (:Is%20Not%20Valid
                _))
            (return false))
          (:Is%20Not%20Valid
            (return false))))
      (:Is%20Not%20Valid
        (return false)))
    (else
      (return false))))
